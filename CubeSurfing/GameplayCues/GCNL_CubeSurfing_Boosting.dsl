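; Looping gameplay cue notify for Cube Surfing boost, driving the boost's audio, particle and screen effects on the target player pawn.
; On loop start it turns on boost audio through the current weapon's SpatialTravel sound library component, then caches and rotates one of the spawned particle components.
; For locally controlled pawns only, it also finds or adds the DerezSurfing boost screen-FX camera modifier and enables it.
; On removal it switches off the boost audio and disables that camera modifier.
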(event EventOnLoopingStart(EitherVFX-preferthisversion)
  (params (MyTarget :type Actor%20Object%20Reference) (Parameters :type Gameplay%20Cue%20Parameters%20Structure%20%28by%20ref%29) (ParticleComponents :type Array%20of%20FXSystem%20Component%20Object%20References) (AudioComponents :type Array%20of%20Audio%20Component%20Object%20References))
  (|CallParentFunction MyTarget Parameters ParticleComponents AudioComponents)
  (bind _asfort_player_pawn (Utilities|Casting|CastToFortPlayerPawn MyTarget)
    (:then
      (MovementLoops|OnBoostChange (Actor|GetComponentbyClass (Weapon|GetCurrentWeapon _asfort_player_pawn) "/Game/Sounds/Fort_Audio_Blueprints/SpatialTravelAudioComponent/SpatialTravel_SoundLibraryComponent.SpatialTravel_SoundLibraryComponent_C") true))
    (:CastFailed
      _))
  (bind _output_get (Variables|Default|SetCachedNiagara (Utilities|Array|Get(acopy) ParticleComponents 1)))
  (Transformation|SetRelativeRotation _output_get "0.000000,90.000000,0.000000" false false)
  (if (Pawn|IsLocallyControlled _asfort_player_pawn)
    (bind _outmodifier (Class|WeaponFunctionLibrary|FindorAddCameraModifierforPawn (Variables|Default|GetCamera%20Modifier%20Class) MyTarget))
    (bind _asbp_camera_modifier_screen_fx_derez_surfing_on_boost (Utilities|Casting|CastToBP_CameraModifier_ScreenFX_DerezSurfing_OnBoost _outmodifier))
    (CameraModifier|EnableModifier _asbp_camera_modifier_screen_fx_derez_surfing_on_boost)))

(event EventOnRemoval(EitherVFX-preferthisversion)
  (params (MyTarget :type Actor%20Object%20Reference) (Parameters :type Gameplay%20Cue%20Parameters%20Structure%20%28by%20ref%29) (ParticleComponents :type Array%20of%20FXSystem%20Component%20Object%20References) (AudioComponents :type Array%20of%20Audio%20Component%20Object%20References) (BurstCameraShakeInstance :type Camera%20Shake%20Base%20Object%20Reference) (BurstDecalInstance :type Decal%20Actor%20Object%20Reference))
  (bind _asfort_player_pawn (Utilities|Casting|CastToFortPlayerPawn MyTarget)
    (:then
      (MovementLoops|OnBoostChange (Actor|GetComponentbyClass (Weapon|GetCurrentWeapon _asfort_player_pawn) "/Game/Sounds/Fort_Audio_Blueprints/SpatialTravelAudioComponent/SpatialTravel_SoundLibraryComponent.SpatialTravel_SoundLibraryComponent_C") false))
    (:CastFailed
      _))
  (if (Pawn|IsLocallyControlled _asfort_player_pawn)
    (bind _outmodifier (Class|WeaponFunctionLibrary|FindorAddCameraModifierforPawn (Variables|Default|GetCamera%20Modifier%20Class) MyTarget))
    (bind _asbp_camera_modifier_screen_fx_derez_surfing_on_boost (Utilities|Casting|CastToBP_CameraModifier_ScreenFX_DerezSurfing_OnBoost _outmodifier))
    (CameraModifier|DisableModifier _asbp_camera_modifier_screen_fx_derez_surfing_on_boost false)))

(fn ConstructionScript)
