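; Looping gameplay cue that drives the VFX while a Fortnite pawn rides the CubeSurfing board weapon.
; On application it binds weapon events, which toggle boost, water and land parameters on two cached Niagara components and map ammo to a normalized depletion value.
; Loop start also spawns an attached wash force actor behind an effects quality check, with timeline-driven strength.
; Removal unbinds everything and fades that force out before destroying it.
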(event EventOnApplication(EitherVFX-preferthisone)
  (params (MyTarget :type Actor%20Object%20Reference) (Parameters :type Gameplay%20Cue%20Parameters%20Structure%20%28by%20ref%29) (ParticleComponents :type Array%20of%20FXSystem%20Component%20Object%20References) (AudioComponents :type Array%20of%20Audio%20Component%20Object%20References) (BurstCameraShakeInstance :type Camera%20Shake%20Base%20Object%20Reference) (BurstDecalInstance :type Decal%20Actor%20Object%20Reference))
  (bind _asfort_pawn (Utilities|Casting|CastToFortPawn MyTarget))
  (if (Utilities|IsA (Variables|/Script/FortniteGame.FortPawn|GetCurrentWeapon _asfort_pawn) "/CubeSurfing/Gameplay/Actors/B_CubeSurfing_Weapon.B_CubeSurfing_Weapon_C")
    (bind _output_get (Variables|Default|SetCachedWeapon (Variables|/Script/FortniteGame.FortPawn|GetCurrentWeapon _asfort_pawn)))
    (EventDispatchers|BindEventtoOnLocalAmmoChanged _output_get (EventDispatchers|CreateEvent :!event UpdateBoardAmmo))
    (bind _asb_cube_surfing_weapon (Utilities|Casting|CastToB_CubeSurfing_Weapon (Variables|/Script/FortniteGame.FortPawn|GetCurrentWeapon _asfort_pawn)))
    (bind _output_get_2 (Variables|Default|SetAs%20B%20Cube%20Surfing%20Weapon _asb_cube_surfing_weapon))
    (Default|BindEventtoBoosted _output_get_2 (EventDispatchers|CreateEvent :!event Custom|OnBoosted))
    (Default|BindEventtoNotBoosted _output_get_2 (EventDispatchers|CreateEvent :!event Custom|NotBoosted))
    (Default|BindEventtoOnWater _output_get_2 (EventDispatchers|CreateEvent :!event Custom|OnWater))
    (Default|BindEventtoNotonWater _output_get_2 (EventDispatchers|CreateEvent :!event Custom|NotOnWater))
    (Default|BindEventtoOnLand _output_get_2 (EventDispatchers|CreateEvent :!event Custom|OnLand))
    (Default|BindEventtoOnAir _output_get_2 (EventDispatchers|CreateEvent :!event Custom|OnAir))))

(event EventOnLoopingStart(EitherVFX-preferthisversion)
  (params (MyTarget :type Actor%20Object%20Reference) (Parameters :type Gameplay%20Cue%20Parameters%20Structure%20%28by%20ref%29) (ParticleComponents :type Array%20of%20FXSystem%20Component%20Object%20References) (AudioComponents :type Array%20of%20Audio%20Component%20Object%20References))
  (|CallParentFunction MyTarget Parameters ParticleComponents AudioComponents)
  (bind _output_get (Variables|Default|SetCachedNiagara (Utilities|Array|Get(acopy) ParticleComponents)))
  (Variables|Default|SetCachedNiagaraWater (Utilities|Array|Get(acopy) ParticleComponents 1))
  (Transformation|SetRelativeRotation _output_get "0.000000,90.000000,0.000000" false false)
  (Utilities|FlowControl|Sequence
    (:then_0
      ;[0.2611 1 0.9568 0.5] WIND FORCE
      (Utilities|EffectsQualityCheck 0
        (:Cinematic
          _)
        (:Epic
          _)
        (:High
          _))
      (Utilities|IsValid (Variables|Default|GetPhysical%20Wash%20Force)
        (:Is%20Not%20Valid
          (bind _returnvalue (Game|SpawnActorfromClass MyTarget "/Game/Effects/Niagara/Blueprints/PhsyicsFields/FN_WashForce.FN_WashForce_C" (Transformation|GetActorTransform MyTarget) "Undefined" "MultiplyWithRoot" 0 5.5e+02 (Variables|Default|GetStrength) -999.0 0))
          (bind _output_get_2 (Variables|Default|SetPhysical%20Wash%20Force _returnvalue))
          (Transformation|AttachActorToComponent _output_get_2 (Variables|/Script/Engine.Actor|GetRootComponent MyTarget) "Root" "SnapToTarget" "SnapToTarget" "SnapToTarget" true)
          (bind (_direction _strength) (Timeline|Timeline_0 ReverseFromEnd)
            (:Update
              (Variables|/Game/Effects/Niagara/Blueprints/PhsyicsFields/FN_WashForce.FN_WashForce_C|SetStrength _strength (Variables|Default|GetPhysical%20Wash%20Force)))))) ; Don't spawn the Rotor Wash multiple times
    )))

(event EventOnRemoval(EitherVFX-preferthisversion)
  (params (MyTarget :type Actor%20Object%20Reference) (Parameters :type Gameplay%20Cue%20Parameters%20Structure%20%28by%20ref%29) (ParticleComponents :type Array%20of%20FXSystem%20Component%20Object%20References) (AudioComponents :type Array%20of%20Audio%20Component%20Object%20References) (BurstCameraShakeInstance :type Camera%20Shake%20Base%20Object%20Reference) (BurstDecalInstance :type Decal%20Actor%20Object%20Reference))
  (bind _as_b_cube_surfing_weapon (Variables|Default|GetAs%20B%20Cube%20Surfing%20Weapon))
  (|CallParentFunction MyTarget Parameters ParticleComponents AudioComponents BurstCameraShakeInstance BurstDecalInstance)
  (Ability|Async|EndAction (Variables|Default|GetWaitTagAddAsync))
  (Actor|DestroyActor (Variables|Default|GetRadialForceActor))
  (EventDispatchers|UnbindEventfromOnLocalAmmoChanged (Variables|Default|GetCachedWeapon) (EventDispatchers|CreateEvent :!event UpdateBoardAmmo))
  (Default|UnbindEventfromBoosted _as_b_cube_surfing_weapon (EventDispatchers|CreateEvent :!event Custom|OnBoosted))
  (Default|UnbindEventfromNotBoosted _as_b_cube_surfing_weapon (EventDispatchers|CreateEvent :!event Custom|NotBoosted))
  (Default|UnbindEventfromOnWater _as_b_cube_surfing_weapon (EventDispatchers|CreateEvent :!event Custom|OnWater))
  (Default|UnbindEventfromNotonWater _as_b_cube_surfing_weapon (EventDispatchers|CreateEvent :!event Custom|NotOnWater))
  (Default|UnbindEventfromOnLand _as_b_cube_surfing_weapon (EventDispatchers|CreateEvent :!event Custom|OnLand))
  (Default|UnbindEventfromOnAir _as_b_cube_surfing_weapon (EventDispatchers|CreateEvent :!event Custom|OnAir))
  (bind (_direction _strength) (Timeline|Timeline PlayFromStart)
    (:Update
      (Variables|/Game/Effects/Niagara/Blueprints/PhsyicsFields/FN_WashForce.FN_WashForce_C|SetStrength _strength (Variables|Default|GetPhysical%20Wash%20Force)))
    (:Finished
      (Actor|DestroyActor (Variables|Default|GetPhysical%20Wash%20Force)))))

(event Custom|OnBoosted
  (Effects|Components|ParticleSystem|SetBoolParameter (Variables|Default|GetCachedNiagara) "bIsBoosted" true))

(event Custom|NotBoosted
  (Effects|Components|ParticleSystem|SetBoolParameter (Variables|Default|GetCachedNiagara) "bIsBoosted" false))

(event Custom|NotOnWater
  (Effects|Components|ParticleSystem|SetBoolParameter (Variables|Default|GetCachedNiagaraWater) "bIsOnWater" false))

(event Custom|OnWater
  (Effects|Components|ParticleSystem|SetBoolParameter (Variables|Default|GetCachedNiagaraWater) "bIsOnWater" true))

(event Custom|OnLand
  (Effects|Components|ParticleSystem|SetBoolParameter (Variables|Default|GetCachedNiagara) "blsOnLand" true))

(event Custom|OnAir
  (Effects|Components|ParticleSystem|SetBoolParameter (Variables|Default|GetCachedNiagara) "blsOnLand" false))

(fn GetAmmoPerClip
  (bind (_returnvalue_datatable _returnvalue_rowname) (Utilities|Struct|BreakDataTableRowHandle (WeaponItemDefinition|GetWeaponStatHandle (Data|GetWeaponData (Variables|Default|GetCachedWeapon)))))
  (bind _returnvalue (Utilities|GetDataTableRow _returnvalue_datatable _returnvalue_rowname))
  (return (.ClipSize (Utilities|Struct|BreakFortRangedWeaponStats _returnvalue))))

(fn ToggleFXColorState
  (params (Boosting? :type Boolean))
  (bind _cachedniagara (Variables|Default|GetCachedNiagara))
  (bind _asniagara_particle_system_component (Utilities|Casting|CastToNiagaraComponent _cachedniagara))
  (bind _colorarray (Variables|Default|GetColorArray))
  (bind _fxcomp (LocalVariables|GetFXComp))
  (if (not (Components|IsComponentBeingDestroyed _asniagara_particle_system_component))
    (LocalVariables|SetFXComp _asniagara_particle_system_component)
    (if Boosting?
      (Niagara|SetNiagaraVariable(LinearColor) _fxcomp "CubeColorMin" (Utilities|Array|Get(acopy) _colorarray))
      (Niagara|SetNiagaraVariable(LinearColor) _fxcomp "CubeColorMax" (Utilities|Array|Get(acopy) _colorarray 1))
      (else
        (Niagara|SetNiagaraVariable(LinearColor) _fxcomp "CubeColorMin" (Utilities|Array|Get(acopy) _colorarray 2))
        (Niagara|SetNiagaraVariable(LinearColor) _fxcomp "CubeColorMax" (Utilities|Array|Get(acopy) _colorarray 3))))))

(fn UpdateBoardAmmo
  (params (LocalCount :type Integer) (LocalRemaining :type Integer))
  (bind _cachedniagara (Variables|Default|GetCachedNiagara))
  (Utilities|IsValid _cachedniagara
    (:Is%20Valid
      (Effects|Components|ParticleSystem|SetFloatParameter _cachedniagara "DeRezz_Deplete" (Utilities|NormalizeFloat (Math|Conversions|ToFloat(Integer) LocalCount) 0 (Math|Conversions|ToFloat(Integer) (CallFunction|GetAmmoPerClip)) true)))))

(fn ConstructionScript)
